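; Quest gameplay ability that plays two Season 33 quest speech bubble lines on the owning Fort pawn.
; On activation it adds the speech component to the pawn and starts two filtered EventProcessor listeners.
; Each event starts speech with its conversation tag, then stops its own listener.
; Pawn changes stop active listeners, add a component and re-register both, even ones already played; ending the ability stops active listeners.
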(event EventBP_OnPawnChanged
  (bind _as_fort_pawn (Variables|Default|GetAs%20Fort%20Pawn))
  (Utilities|FlowControl|Sequence
    (:then_0
      (if (Variables|Default|Getevent1%20active?)
        (EventProcessorSubsystem|StopListeningToFilteredEvent (Variables|Default|Getevent1))))
    (:then_1
      (if (Variables|Default|Getevent2%20active?)
        (EventProcessorSubsystem|StopListeningToFilteredEvent (Variables|Default|Getevent2))))
    (:then_2
      (bind _fortpawn (Class|FortQuestBlueprintLibrary|GetFortPawnFromObject (Ability|GetOwningActorfromActorInfo))
        (:Valid
          (bind _output_get (Variables|Default|SetAs%20Fort%20Pawn _fortpawn))
          (bind _returnvalue (Game|AddComponentbyClass _output_get (Variables|Default|GetSpeech%20Component)))
          (Variables|Default|SetAs%20BP%20Quest%20Speech%20Bubble (Utilities|Casting|CastToBP_Quest_SpeechBubble _returnvalue))
          (bind _returnvalue_2 (EventProcessorSubsystem|StartListeningToFilteredEvent 0 _as_fort_pawn (Variables|Default|GetCase0part1) false (EventDispatchers|CreateEvent :!event Custom|Case%200%20part%201%20Event)))
          (Variables|Default|Setevent1 _returnvalue_2)
          (Variables|Default|Setevent1%20active? true)
          (bind _returnvalue_3 (EventProcessorSubsystem|StartListeningToFilteredEvent 0 _as_fort_pawn (Variables|Default|GetCase0part2) false (EventDispatchers|CreateEvent :!event Custom|Case%200%20part%202%20Event)))
          (Variables|Default|Setevent2 _returnvalue_3)
          (Variables|Default|Setevent2%20active? true))))))

(event Ability|EventActivateAbility
  (bind _as_fort_pawn (Variables|Default|GetAs%20Fort%20Pawn))
  (|CallParentFunction)
  (bind _fortpawn (Class|FortQuestBlueprintLibrary|GetFortPawnFromObject (Ability|GetOwningActorfromActorInfo))
    (:Valid
      (bind _output_get (Variables|Default|SetAs%20Fort%20Pawn _fortpawn))
      (bind _returnvalue (Game|AddComponentbyClass _output_get (Variables|Default|GetSpeech%20Component)))
      (Variables|Default|SetAs%20BP%20Quest%20Speech%20Bubble (Utilities|Casting|CastToBP_Quest_SpeechBubble _returnvalue))
      (bind _returnvalue_2 (EventProcessorSubsystem|StartListeningToFilteredEvent 0 _as_fort_pawn (Variables|Default|GetCase0part1) false (EventDispatchers|CreateEvent :!event Custom|Case%200%20part%201%20Event)))
      (Variables|Default|Setevent1 _returnvalue_2)
      (Variables|Default|Setevent1%20active? true)
      (bind _returnvalue_3 (EventProcessorSubsystem|StartListeningToFilteredEvent 0 _as_fort_pawn (Variables|Default|GetCase0part2) false (EventDispatchers|CreateEvent :!event Custom|Case%200%20part%202%20Event)))
      (Variables|Default|Setevent2 _returnvalue_3)
      (Variables|Default|Setevent2%20active? true))))

(event Custom|Case%200%20part%201%20Event
  (Class|FortQuestBlueprintLibrary|GetFortPawnFromObject (Ability|GetOwningActorfromActorInfo)
    (:Valid
      (SpeechBubble|StartSpeech (Variables|Default|GetAs%20BP%20Quest%20Speech%20Bubble) "(TokenStreamVersion=0,TagDictionary=((TagName=\"Conversation.Entry.Season33.WF.Quest.00.part.01\")),QueryTokenStream=(0,1,1,1,0),UserDescription=\"\",AutoDescription=\" ANY( Conversation.Entry.Season33.WF.Quest.00.part.01 )\")")
      (EventProcessorSubsystem|StopListeningToFilteredEvent (Variables|Default|Getevent1))
      (Variables|Default|Setevent1%20active? false))))

(event Custom|Case%200%20part%202%20Event
  (Class|FortQuestBlueprintLibrary|GetFortPawnFromObject (Ability|GetOwningActorfromActorInfo)
    (:Valid
      (SpeechBubble|StartSpeech (Variables|Default|GetAs%20BP%20Quest%20Speech%20Bubble) "(TokenStreamVersion=0,TagDictionary=((TagName=\"Conversation.Entry.Season33.WF.Quest.00.part.02\")),QueryTokenStream=(0,1,1,1,0),UserDescription=\"\",AutoDescription=\" ANY( Conversation.Entry.Season33.WF.Quest.00.part.02 )\")")
      (EventProcessorSubsystem|StopListeningToFilteredEvent (Variables|Default|Getevent2))
      (Variables|Default|Setevent2%20active? false))))

(event Ability|EventOnEndAbility
  (params (bWasCancelled :type Boolean))
  (Utilities|FlowControl|Sequence
    (:then_0
      (if (Variables|Default|Getevent1%20active?)
        (EventProcessorSubsystem|StopListeningToFilteredEvent (Variables|Default|Getevent1))))
    (:then_1
      (if (Variables|Default|Getevent2%20active?)
        (EventProcessorSubsystem|StopListeningToFilteredEvent (Variables|Default|Getevent2))))))
